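; Weapon Blueprint for a BlastBerry pump shotgun that shows its XP evolution level through visuals and audio.
; BeginPlay finds the rifted-weapon XP component, binds its events, reads the evolution level and starts level-up effects, plus a trail at level four.
; Level, XP-change times and upgrade-ready pulsing drive the mesh through custom primitive data, with locally controlled audio and Niagara cues.
; Item upgrade, switch, upgrade-animation and crystal-shatter logic are empty stubs.
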(event EventBeginPlay
  (bind _reloxpcomp (Variables|Default|GetReloXPComp))
  ;[0.1158 1 0.1046] Begin Play - bind to callbacks and initialize.
  (|CallParentFunction)
  (bind _asfort_player_pawn (Utilities|Casting|CastToFortPlayerPawn (Game|GetInstigator))
    (:then
      (if (Pawn|IsLocallyControlled _asfort_player_pawn)
        (GameplayCue|ExecuteGameplayCue%28Burst%29Local (Game|GetInstigator) (Variables|Default|GetGCN_ReformCrystals) (Ability|GameplayCue|MakeGameplayCueParameters 0.0))
        (else
          _))
      (CallFunction|MIDPlayUpgradeAnimation)
      (bind _output_get (Variables|Default|SetReloXPComp (Actor|GetComponentbyClass "/BlastBerryWeapons/Gameplay/C_RiftedWeapons_XPComponent.C_RiftedWeapons_XPComponent_C")))
      (Utilities|FlowControl|Sequence
        (:then_0
          (Utilities|IsValid _output_get
            (:Is%20Valid
              (Default|BindEventtoOnPowerUpReady _output_get (EventDispatchers|CreateEvent :!event Custom|OnPowerUpReady))
              (Variables|Default|SetEvolutionLevel (MultiItem|GetSelectedItemIndex _output_get))
              (Default|BindEventtoOnXPChanged _output_get (EventDispatchers|CreateEvent :!event Custom|On%20XP%20Change))
              (Default|BindEventtoOnItemUpgraded _reloxpcomp (EventDispatchers|CreateEvent :!event Custom|OnItemUpgrade))
              (Default|BindEventtoUpgradeSwitch _reloxpcomp (EventDispatchers|CreateEvent :!event Custom|ItemSwitch)))))
        (:then_1
          (CallFunction|MIDChangeXP 0.0 (Variables|Default|GetEvolutionLevel))
          (Components|Activation|Activate (Variables|Default|GetNiagara_LevelUp) false)
          (if (== (Variables|Default|GetEvolutionLevel) 4)
            (Components|Activation|Activate (Variables|Default|GetNS_Morphite_WeaponTrail) false)
            (else
              _))
          (if (Pawn|IsLocallyControlled (Variables|/Script/Engine.Actor|GetInstigator self))
            (Audio|SpawnSoundAttached (Variables|Default|GetSFX_Upgrade) (Variables|Default|GetMSS_Morphite_LevelUp_Weapon))))))
    (:CastFailed
      _)))

(event EventTick
  (params (DeltaSeconds :type Float%20%28single-precision%29)))

(event Custom|CustomEvent
  (params (Selection :type EFortReloadFXState%20Enum :default "ReloadStart"))
  ;[0.1158 1 0.1046] Begin Play - bind to callbacks and initialize.
  (Default|BindEventtoOnXPChanged (Variables|Default|GetReloXPComp) (EventDispatchers|CreateEvent :!event Custom|On%20XP%20Change)))

(event Custom|OnPowerUpReady
  (params (CurrentItemIndex :type Integer) (PendingItemIndex :type Integer))
  ;[1 0.1749 0.7797] Rarity based audio
  (if (Pawn|IsLocallyControlled (Variables|/Script/Engine.Actor|GetInstigator self))
    (Audio|PlaySound2D (Variables|Default|GetSFX_UpgradeReady))
    (Audio|Modulation|ActivateControlBusMix (Variables|Default|GetControl%20Bus%20Mix))
    (Components|Activation|Activate (Variables|Default|GetNS_ReloChrome_LevelUpReady_Weapon) false)
    (Components|Activation|Activate (Variables|Default|GetNS_ReloChrome_LevelUpReady_ADS) false)
    (Rendering|Material|SetCustomPrimitiveDataFloat (Variables|Visual|GetWeaponMesh) 4 1.0) ; level up pulsing
  ))

(event EventPlayWeaponFireCustomAudio
  (params (bPersistentFire :type Boolean) (bSecondaryFire :type Boolean)))

(event Custom|AnimNotify_Upgrade_ShatterCrystals
  ;[1 1 1] Event that plays when crystals shatter in upgrade animation
  (CallFunction|MIDShatterCrystals))

(event Custom|On%20XP%20Change
  (params (XPDelta :type Float%20%28double-precision%29) (CurrentXPPercentage :type Float%20%28double-precision%29))
  ;[1 1 1] Event that plays when crystals shatter in upgrade animation
  (Rendering|Material|SetCustomPrimitiveDataFloat (Variables|Visual|GetWeaponMesh) 3 (Utilities|Time|GetGameTimeinSeconds)))

(event Custom|OnItemUpgrade
  (params (CurrentItemIndex :type Integer)))

(event Custom|ItemSwitch)

(event EventMuzzlePlayReloadFX
  (params (Selection :type EFortReloadFXState%20Enum :default "ReloadStart"))
  (|CallParentFunction Selection))

(fn MIDChangeXP
  (params (XP%20Normalized :type Float%20%28double-precision%29) (NewParam :type Integer))
  (bind _returnvalue (Math|Conversions|ToFloat(Integer) NewParam))
  (bind _weaponmesh (Variables|Visual|GetWeaponMesh))
  (Rendering|Material|SetCustomPrimitiveDataFloat _weaponmesh 1 _returnvalue)
  (Rendering|Material|SetCustomPrimitiveDataFloat _weaponmesh 2 (Utilities|Time|GetGameTimeinSeconds))
  (Niagara|SetNiagaraVariable(Float) (Variables|Default|GetNiagara_LevelUp) "WeaponPhase" _returnvalue))

(fn MIDPlayUpgradeAnimation)

(fn MIDShatterCrystals)

(fn ConstructionScript
  (|CallParentFunction))
